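# Translucent unlit Niagara sprite and mesh material for the core of a WinterDecay teleport portal.
# Twisted polar UVs drive panning noise that blends dark and light wave colors inside sphere-mask circles with a bright rim, while Perlin noise builds warped UVs.
# A Filling static switch selects this disc or a hue-shifted particle-color variant, with polar dot textures also defined.
# Particle alpha, depth fade (skipped on mobile) and camera-distance fade shape opacity.
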Material /WinterDecay/Assets/FX/Teleport/Materials/M_WinterDecay_PortalCore
graph {
 properties {
  bAutomaticallySetUsageInEditor=false
  bUsedWithNiagaraMeshParticles=true
  bUsedWithNiagaraSprites=true
  blendMode=BLEND_Translucent
  shadingModel=MSM_Unlit
  usedShadingModels=Unlit
 }
 nodes [
  Constant() {r=5}
  n1 = TextureCoordinate()
  n2 = TextureSampleParameter2D(UVs=Panner(Coordinate=Multiply(A=@n1) {constB=0.66}) {speedX=0.15, speedY=-0.3}) {parameterName=Tex2, samplerType=SAMPLERTYPE_Masks, texture=/Game/Characters/Player/Male/Large/Bodies/M_LRG_Clash_V/Textures/Simple_Perlin_M}
  LinearInterpolate(Alpha=Power(Base=@n2.R)) {constA=0.5, constB=2}
  n3 = VectorParameter() {parameterName=WavesDark, defaultValue=<rgba 0.12 0.2333 0.8 1>}
  n4 = VectorParameter() {parameterName=WavesLight, defaultValue=<rgba 0.02 2 0.68 1>}
  n5 = MaterialFunctionCall!(UV=TextureCoordinate(), Twist Center, Twist Edges=Constant() {r=0.5}) {materialFunction=/Game/Accessories/FORT_Backpacks/Backpack_F_MED_Celestial_Glow/FX/Materials/MF_PolarUV}
  n6 = MaterialFunctionCall(Float2=@n5.Polar UVs) {materialFunction=&BreakOutFloat2Components}
  n7 = MaterialFunctionCall(Float2=@n5.Polar UVs) {materialFunction=&BreakOutFloat2Components}
  n8 = AppendVector(A=Constant() {r=1}, B=LinearInterpolate(Alpha=@n7.G) {constA=6})
  n9 = AppendVector(A=@n7.R, B=Power(Base=@n7.G) {constExponent=0.5})
  n10 = TextureSampleParameter2D!(UVs=Panner(Coordinate=Multiply(A=@n8, B=@n9)) {speedX=0.1, speedY=0.066}, "DDX(UVs)"=@n5.DDX, "DDY(UVs)"=@n5.DDX, Apply View MipBias) {
   group=Main Texture
   mipValueMode=TMVM_Derivative
   parameterName=TEx1
   samplerType=SAMPLERTYPE_LinearGrayscale
   texture=/Game/Characters/Player/Male/Medium/Bodies/M_MED_Citadel/FX/T_Citadel_Noise_Dissolve_Handpaint
  }
  n11 = Multiply(A=@n6.G, B=@n10.R)
  n12 = Saturate(None=Multiply(A=Power(Base=@n11) {constExponent=1.6}) {constB=2})
  LinearInterpolate(A=@n3.RGB, B=@n4.RGB, Alpha=@n12)
  n13 = SphereMask(A=NamedRerouteUsage(), B=Constant() {r=0.5}, Radius=Constant() {r=0.5}) {hardnessPercent=3}
  n14 = Power(Base=@n13) {constExponent=4}
  LinearInterpolate(A=@n10.R, Alpha=Saturate(None=Multiply(A=@n14) {constB=3}))
  n15 = Power(Base=@n6.G) {constExponent=0.4}
  n16 = LinearInterpolate(A=Constant3Vector() {constant=<rgba 0.15 0.175 0.3 1>}, B=@n3.RGB, Alpha=Saturate(None=Multiply(A=@n15)))
  n17 = Power(Base=@n12) {constExponent=1}
  n18 = Multiply(A=LinearInterpolate(A=Add(A=@n16) {constB=0}, B=@n4.RGB, Alpha=Saturate(None=@n17)), B=LinearInterpolate(Alpha=@n13) {constA=1, constB=0.5})
  n19 = Power(Base=OneMinus(None=@n13)) {constExponent=3}
  n20 = NamedRerouteUsage()
  n21 = Multiply(A=Multiply(A=Multiply(A=@n19), B=Power(Base=@n20) {constExponent=1.4})) {constB=12}
  n22 = Constant3Vector() {constant=<rgba 13.6 0.8 20 1>}
  n23 = Constant3Vector() {constant=<rgba 0.4 16.9 40 1>}
  n24 = MaterialFunctionCall(Input=NamedRerouteUsage(), Input Low=Constant(), Input High=Constant() {r=0.02}) {materialFunction=&RemapValue_0-1}
  n25 = NamedRerouteUsage()
  n26 = Constant() {r=0.5}
  n27 = SphereMask(A=@n25, B=@n26, Radius=Constant() {r=0.51}) {hardnessPercent=0}
  n28 = SphereMask(A=Add(A=NamedRerouteUsage()) {constB=0}, B=Constant() {r=0.5}, Radius=ScalarParameter() {defaultValue=0.49, parameterName=FillingRadius}) {hardnessPercent=1}
  n29 = LinearInterpolate(Alpha=@n28)
  n30 = Multiply(A=Subtract(A=@n27, B=Saturate(None=@n29))) {constB=5}
  n31 = Add(A=@n24.Result, B=Saturate(None=Power(Base=@n30) {constExponent=3}))
  n32 = ParticleColor()
  n33 = Multiply(A=LinearInterpolate(A=Add(A=@n18, B=@n21), B=LinearInterpolate(A=@n22, B=@n23, Alpha=@n24.Result), Alpha=Saturate(None=@n31)), B=@n32.RGB)
  n34 = ScalarParameter() {parameterName=Glow}
  n35 = MaterialFunctionCall(Hue Shift Percentage=DynamicParameter() {paramNames [HueShift, DotsPower, PanSpeed, WarpIntensity]}.HueShift, Texture=@n32.RGB) {materialFunction=&HueShift}
  n36 = Multiply(A=NamedRerouteUsage()) {constB=333}
  MaterialFunctionCall(Emissive=StaticSwitchParameter(True=LinearInterpolate(A=@n33, B=@n32.RGB, Alpha=@n34), False=Add(A=@n35.Result, B=@n36)) {defaultValue=true, parameterName=Filling}) {materialFunction=&MF_FX_EmissiveAdjustments}
  n37 = MaterialFunctionCall(Input=NamedRerouteUsage(), Input Low=Constant(), Input High=Constant() {r=1}) {materialFunction=&RemapValue_0-1}
  n38 = Power(Base=Desaturation(None=NamedRerouteUsage(), Fraction=Constant() {r=1})) {constExponent=1}
  n39 = Multiply(A=@n37.Result, B=SmoothStep(Value=Saturate(None=@n38)) {constMax=0.2})
  n40 = DynamicParameter() {paramNames [HueShift, DotsPower, PanSpeed, WarpIntensity]}
  n41 = Multiply(
   A=ParticleColor().A
   B=StaticSwitchParameter(True=NamedRerouteUsage(), False=LinearInterpolate(A=@n37.Result, B=@n39, Alpha=@n40.DotsPower)) {defaultValue=true, parameterName=Filling}
  ) {constA=1}
  n42 = DepthFade(Opacity=@n41, FadeDistance=ScalarParameter() {defaultValue=100, parameterName=DepthFadeDistance})
  Multiply(
   A=ShadingPathSwitch(Default=@n42, Deferred=@n42, Mobile=@n41) {desc=Migrated from FeatureLevelSwitch}
   B=MaterialFunctionCall!(Distance=ScalarParameter() {defaultValue=200, parameterName=CameraFadeDistance}) {materialFunction=/Game/Effects/Fort_Effects/Materials/_Functions/MF_CameraDistanceFade}.Result
  )
  n43 = Power(Base=@n2.R) {constExponent=0.2}
  Multiply(A=Multiply(A=NamedRerouteUsage(), B=Saturate(None=@n43)), B=ScalarParameter() {defaultValue=0.05, parameterName=SmallNoiseIntensity}) {constB=0.1}
  n44 = MaterialFunctionCall(
   UV=TextureCoordinate()
   Twist Center=ScalarParameter() {parameterName=DotsTwistCenter}
   Twist Edges=ScalarParameter() {defaultValue=0.33, parameterName=DotsTwistEdge}
  ) {materialFunction=&MF_PolarUV}
  n45 = Add(A=ParticleRandom(), B=@n44.Polar UVs)
  n46 = Multiply(A=Constant2Vector() {g=0.33, r=1}, B=ScalarParameter() {defaultValue=2, parameterName=DotsScale}) {constB=6}
  n47 = ScalarParameter() {defaultValue=0.22, parameterName=DotsPanSpeedX}
  n48 = ScalarParameter() {defaultValue=-0.066, parameterName=DotsPanSpeedY}
  n49 = TextureSampleParameter2D(UVs=Panner(Coordinate=Multiply(A=@n45, B=@n46), Speed=AppendVector(A=@n47, B=@n48)) {speedX=0.77}, "DDX(UVs)"=@n44.DDX, "DDY(UVs)"=@n44.DDX) {group=Main Texture, mipValueMode=TMVM_Derivative, parameterName=DotsPolar, texture=/Game/Effects/Fort_Effects/Textures/Tile/T_Dots_Small}
  NamedRerouteDeclaration(None=@n49.RGB) {name=RainbowDots, nodeColor=<rgba 1 0.1266 0.0286 1>}
  n50 = TextureCoordinate()
  n51 = AppendVector(A=ScalarParameter() {defaultValue=0.1, parameterName=WarpScaleX}, B=ScalarParameter() {defaultValue=0.33, parameterName=WarpScaleY})
  n52 = TextureSampleParameter2D!(UVs=Panner(Coordinate=Multiply(A=@n50, B=@n51) {constB=0.2}) {speedX=0.03, speedY=0.07}, Apply View MipBias) {
   parameterName=WarpNoise
   samplerType=SAMPLERTYPE_Masks
   texture=&Simple_Perlin_M
  }
  n53 = Multiply(A=@n52.R, B=NamedRerouteUsage())
  n54 = Multiply(
   A=ScalarParameter() {defaultValue=0.33, parameterName=WarpIntensity}
   B=DynamicParameter() {paramNames [HueShift, DotsPower, PanSpeed, WarpIntensity]}.WarpIntensity
  )
  NamedRerouteDeclaration(None=Add(A=TextureCoordinate(), B=Multiply(A=@n53, B=@n54) {constB=0.33})) {name=Warped_UVs, nodeColor=<rgba 0.5469 0 1 1>}
  n55 = MaterialFunctionCall(Float2=TextureCoordinate()) {materialFunction=&BreakOutFloat2Components}
  NamedRerouteDeclaration(None=AppendVector(A=LinearInterpolate(Alpha=@n55.R) {constA=-1}, B=LinearInterpolate(Alpha=@n55.G) {constA=-1})) {name="UV[-1,1]", nodeColor=<rgba 0 1 0.2266 1>}
  n56 = LinearInterpolate(Alpha=SphereMask(A=@n25, B=@n26, Radius=Constant() {r=0.5}) {hardnessPercent=10})
  n57 = LinearInterpolate(Alpha=@n28) {constB=3}
  NamedRerouteDeclaration(None=Subtract(A=Saturate(None=@n56), B=Saturate(None=@n57))) {name=EdgeCircle, nodeColor=<rgba 1 0 0.0937 1>}
  n58 = LinearInterpolate(Alpha=@n27) {constB=10}
  n59 = Multiply(A=@n27)
  n60 = ScalarParameter() {defaultValue=1, parameterName=GlowPower}
  NamedRerouteDeclaration(None=LinearInterpolate(A=Saturate(None=@n58), B=Power(Base=@n59, Exp=@n60), Alpha=ScalarParameter() {parameterName=Glow})) {name=OpacityMask, nodeColor=<rgba 1 0.8049 0.094 1>}
  n61 = TextureSampleParameter2D!(UVs=Panner(Coordinate=@n5.Polar UVs) {speedY=-0.33}, "DDX(UVs)"=@n5.DDX, "DDY(UVs)"=@n5.DDX, Apply View MipBias) {
   group=Main Texture
   mipValueMode=TMVM_Derivative
   parameterName=Tex3
   texture=/Game/Effects/Fort_Effects/Textures/Tile/T_ShieldCaustics_Tile_Chromatic
  }
  n62 = Multiply(A=@n6.G, B=@n61.RGB)
  Reroute(None=Multiply(A=Multiply(A=@n62, B=@n52.R)) {constB=0.7})
  SphereMask(A=NamedRerouteUsage(), B=Constant() {r=0.5}, Radius=Constant() {r=0.55}) {hardnessPercent=0}
 ]
}
