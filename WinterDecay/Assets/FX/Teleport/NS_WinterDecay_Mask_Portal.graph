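# Niagara system for the WinterDecay mask teleport portal: a glowing core that scales open, surrounded by waves, helix meshes and sparkles.
# MarkerCore layers meshes and a light, with time-of-day tints on several layers.
# ExtraWaves pulse, Helix meshes spin and sink, and GPU Sparkles fall; EnergyRadial is disabled.
# Shared system values drive scale, fade-in and destroyed-state fade, while camera-distance fades and distance culling limit visibility.
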NiagaraSystem /WinterDecay/Assets/FX/Teleport/NS_WinterDecay_Mask_Portal
dependencies {
 usedDataInterfaces [&NiagaraDataInterfaceCurve, &NiagaraDataInterfaceColorCurve, &NiagaraDataInterfaceCamera]
 usedDynamicInputs [&SpawnGroup]
 usedModules [
  &self:SystemSpawnScript.SystemScriptSource.SystemScriptGraph.NiagaraNodeAssignment_1.SetVariables_#1
  &self:SystemSpawnScript.SystemScriptSource.SystemScriptGraph.NiagaraNodeAssignment_0.SetVariables_#2
  &self:SystemSpawnScript.SystemScriptSource.SystemScriptGraph.NiagaraNodeAssignment_2.SetVariables_#3
  &SystemState
  &self:SystemSpawnScript.SystemScriptSource.SystemScriptGraph.NiagaraNodeAssignment_3.SetVariables_#4
  &self:SystemSpawnScript.SystemScriptSource.SystemScriptGraph.NiagaraNodeAssignment_4.SetVariables_#5
  &self:SystemSpawnScript.SystemScriptSource.SystemScriptGraph.NiagaraNodeAssignment_5.SetVariables_#6
 ]
}
emitters {
 EnergyRadial {
  bEnabled=false
  emitterUpdateScript [
   &EmitterState(Life Cycle Mode=::NewEnumerator1, Loop Behavior=::NewEnumerator1, Loop Duration Mode=::NewEnumerator1, UseLoopDelay=false)
   &SpawnBurst_Instantaneous(Use Spawn Probability=false, UseLoopCountLimit=false)
   &SpawnRate(SpawnRate=2, Use Spawn Probability=false) {enabled=false}
   /SetVariables(
    Emitter.GlowColor:LinearColor=<rgba 2 9.5 20 1>
    Emitter.GlowPosition:NiagaraPosition=<0 -10 0>
    Emitter.GlowSize:Vector2f=&Multiply_Vector2D_ByFloat(Float=35, Vector2D=<1 1>)
   ) {isSetParameters=true}
  ]
  particleSpawnScript [
   &InitializeParticle(
    Color=<rgba 0.36 1.8343 3 1>
    Lifetime=3
    Mesh Scale=&Multiply_VectorByFloat(Vector=<1 1 1.3>, Float=0.6)
    Mesh Scale Mode=::NewEnumerator4
    Position Offset=$System.SystemOffset
    Sprite Rotation Angle=&MergeHelper_FloatSwitch(0=0, 1=0.5, 2=?, 3=?, Index=&RandomRangeInt(Minimum=0, Maximum=1), Num=::NewEnumerator0)
    Sprite Rotation Mode=::NewEnumerator1
    Sprite Size=&Multiply_Vector2D_ByFloat(Float=&Multiply_Float(A=$System.FXScale, B=160), Vector2D=<0.8 1.25>)
    Sprite Size Mode=::NewEnumerator4
    UsePositionOffset=true
    Write Material Random=true
   )
   &InitialMeshOrientation(Mesh Orientation Mode=::NewEnumerator2, Use Rotation Vector=false)
   &CameraOffset(Camera Offset Amount=3) {module=CameraOffset001}
  ]
  particleUpdateScript [
   &ParticleState(Kill Particles When Lifetime Has Elapsed=false, Let Infinitely Lived Particles Die When Emitter Deactivates=true)
   &ScaleSpriteSize(Uniform Curve Sprite Scale=DataInterface() {propertyValues {Curve {keys [{tangentMode=RCTM_User, time=0, value=0.8}, {tangentMode=RCTM_User, time=1, value=1}]}}})
   &ScaleMeshSize(Scale Factor=&VectorFromFloat(Value=&FloatFromCurve(FloatCurve=DataInterface() {propertyValues {Curve {keys [<at 0 0.1>,<at 0.3 0.7>,<at 1 1>]}}}))) {enabled=false}
   &ScaleColor(
    Scale Alpha=&FloatFromCurve(
     FloatCurve=DataInterface() {propertyValues {Curve {keys [
         {interpMode=RCIM_Cubic, tangentMode=RCTM_User, time=0, value=0}
         {arriveTangent=-0.1677, interpMode=RCIM_Cubic, leaveTangent=-0.1677, tangentMode=RCTM_User, time=0.45, value=1}
         {interpMode=RCIM_Cubic, tangentMode=RCTM_User, time=1, value=0}
        ]}}}
     Scale Curve=$System.CurveDestroyed
    )
    Scale RGB=&MakeVectorFromLinearColorRGB(Color=&ColorFromCurve(ColorCurve=DataInterface() {propertyValues {
        AlphaCurve {keys [<at 0 1 RCIM_Cubic>]}
        BlueCurve {keys [<at 0 15>,<at 0.3636 1.5>,<at 0.9771 1>]}
        GreenCurve {keys [<at 0 15>,<at 0.3636 1.5>,<at 0.9771 0.16>]}
        RedCurve {keys [<at 0 9>,<at 0.3636 0.9>,<at 0.9771 0.1>]}
       }}))
    ScaleA=true
    ScaleRGB=true
   ) {enabled=false}
   /SetVariables(Particles.GlowColor=&MultiplyLinearColor(
     Linear Color=$Particles.Color
     Scale Factor=&MakeVector4_FromLinearColor(LinearColor=<rgba 4 2.785 2.632 1>)
    )) {enabled=false, isSetParameters=true}
   &ScaleColor(
    Scale Alpha=$System.CurveDestroyed
    Scale RGB=&MakeVectorFromLinearColorRGB(Color=<rgba 2.7 4.35 6 1>)
    ScaleA=true
    ScaleRGB=true
   ) {module=ScaleColor001}
  ]
  rendererClasses [&NiagaraSpriteRendererProperties, &NiagaraMeshRendererProperties]
  renderers [
   {rendererClass=&NiagaraSpriteRendererProperties}
   {rendererClass=&NiagaraMeshRendererProperties, rendererIndex=1}
  ]
 }
 ExtraWaves {
  emitterSpawnScript [/SetVariables(Emitter.float:NiagaraFloat=&RandomRangeFloat(Minimum=0, Maximum=500)) {isSetParameters=true}]
  emitterUpdateScript [
   &EmitterState(
    Enable Distance Culling=true
    Life Cycle Mode=::NewEnumerator1
    Loop Behavior=::NewEnumerator1
    Loop Duration Mode=::NewEnumerator1
    MaxDistance=3000
    Module.bUseMaxDistance=true
    Module.bUseMinDistance=false
    Scalability Mode=::NewEnumerator1
    UseLoopDelay=false
   )
   &SpawnRate(Use Spawn Probability=false)
  ]
  particleSpawnScript [
   &InitializeParticle(
    Color=<rgba 0.5 25.25 50 1>
    Lifetime=1.5
    Mesh Scale=&Multiply_VectorByFloat(Vector=<1.3 1 1>, Float=&Multiply_Float(A=0.53, B=$System.FXScale))
    Mesh Scale Mode=::NewEnumerator4
    Position Offset=$System.SystemOffset
    Sprite Rotation Mode=::NewEnumerator2
    UsePositionOffset=true
    Write Material Random=true
   )
   &InitialMeshOrientation(
    Mesh Axis To Orient=<0 0 1>
    Mesh Orientation Mode=::NewEnumerator2
    Rotation=&MergeHelper_VectorSwitch(0=<0 0 1>, 1=<0 0 -1>, 2=?, 3=?, Index=&RandomRangeInt(Minimum=0, Maximum=1), Num=::NewEnumerator0)
    Use Rotation Vector=true
   )
  ]
  particleUpdateScript [
   &ParticleState()
   &Timeline(Initial Start Time:NiagaraFloat=$Emitter.float, Play Rate=0.1, Timeline End=100000)
   &DynamicMaterialParameters(
    Index 0 Param 1=&MakeFloatFromVector(Vector=$Particles.Scale, Channel=::NewEnumerator0)
    Index 0 Param 2=1
    Index 0 Param 3=1
    Index 0 Param 4=1
    Index 1 Param 2=1
    Module.Index0Float0Write=true
    Module.Index0Float1Write=true
    Module.Index0Float2Write=true
    Module.Index0Float3Write=true
    Module.Index1Float0Write=false
    Module.Index1Float1Write=true
    Module.Index1Float2Write=true
    Module.Index1Float3Write=false
    Module.Index2Float0Write=true
    Module.Index2Float1Write=true
    Module.Index2Float2Write=true
    Module.Index2Float3Write=false
    Write Parameter Index 1=::NewEnumerator1
    Write Parameter Index 2=::NewEnumerator1
   )
   &ScaleMeshSize(Scale Factor=&VectorFromFloat(Value=&FloatFromCurve(FloatCurve=DataInterface() {propertyValues {Curve {keys [
          {interpMode=RCIM_Cubic, leaveTangent=0.1968, tangentMode=RCTM_Break, time=0, value=0.72}
          {arriveTangent=0.0729, interpMode=RCIM_Cubic, tangentMode=RCTM_Break, time=1, value=1}
         ]}}})))
   &ScaleColor(
    Scale Alpha=&FloatFromCurve(
     FloatCurve=DataInterface() {propertyValues {Curve {keys [
         {interpMode=RCIM_Cubic, tangentMode=RCTM_User, time=0.1, value=0}
         {arriveTangent=-1.107, interpMode=RCIM_Cubic, leaveTangent=-1.107, tangentMode=RCTM_User, time=0.5, value=1}
         {interpMode=RCIM_Cubic, tangentMode=RCTM_User, time=1, value=0}
        ]}}}
     Scale Curve=&RandomRangeFloat(Minimum=1, Maximum=2)
    )
    Scale RGB=&VectorFromFloat(Value=&FloatFromCurve(
      FloatCurve=DataInterface() {propertyValues {Curve {keys [{interpMode=RCIM_Cubic, tangentMode=RCTM_User, time=0.3, value=1}, {interpMode=RCIM_Cubic, tangentMode=RCTM_User, time=1, value=0.25}]}}}
      Scale Curve=1.6
     ))
    ScaleA=true
    ScaleRGB=false
   )
   &ScaleColor(Scale Alpha=&Multiply_Float(A=$System.CurveDestroyed, B=$System.CamDistanceAlpha), ScaleA=true, ScaleRGB=true) {module=ScaleColor002}
   /SetVariables(Particles.ColorSparkles:LinearColor=&MultiplyLinearColor(
     Linear Color=$Particles.Initial.Color
     Scale Factor=&MakeVector4FromVectorAndFloat(
      "Vector (RGB)"=<1 1 1>
      "Float (Alpha)"=&FloatFromCurve(
       FloatCurve=DataInterface() {propertyValues {Curve {keys [
           <at -0.0108 0 RCIM_Cubic>
           {interpMode=RCIM_Cubic, tangentMode=RCTM_User, time=0.25, value=1}
           {interpMode=RCIM_Cubic, tangentMode=RCTM_User, time=0.66, value=0}
          ]}}}
       Scale Curve=&Multiply_Float(A=$System.CurveDestroyed, B=$System.CamDistanceAlpha)
      )
     )
    )) {isSetParameters=true}
   &CameraOffset(Camera Offset Amount=-15)
  ]
  rendererClasses [&NiagaraMeshRendererProperties]
  renderers [
   {rendererClass=&NiagaraMeshRendererProperties, rendererIndex=0}
   {rendererClass=&NiagaraMeshRendererProperties, rendererIndex=1}
  ]
 }
 Helix {
  emitterUpdateScript [
   &EmitterState(
    Enable Distance Culling=true
    Life Cycle Mode=::NewEnumerator1
    Loop Behavior=::NewEnumerator1
    Loop Duration Mode=::NewEnumerator1
    MaxDistance=4000
    Module.bUseMaxDistance=true
    Module.bUseMinDistance=false
    Scalability Mode=::NewEnumerator1
    UseLoopDelay=false
   )
   &SpawnBurst_Instantaneous(Spawn Count=3, Use Spawn Probability=false, UseLoopCountLimit=true) {enabled=false}
   &SpawnRate(SpawnRate=3, Use Spawn Probability=false)
  ]
  particleSpawnScript [
   &PartitionParticles() {enabled=false}
   &InitializeParticle(
    Lifetime=&RandomRangeFloat(Minimum=1.5, Maximum=2)
    Mesh Scale=&Multiply_VectorByFloat(Vector=&RandomRangeVector(Minimum=<2 2 0.2>, Maximum=<2.5 2.5 0.3>), Float=-0.55)
    Mesh Scale Mode=::NewEnumerator4
    Position Offset=&AddVectorToPosition(Position=Linked() {linkedVariable {name=System.SystemOffset, type=Vector3f}}, Vector=<0 0 -35>)
    Position Offset Coordinate Space=::World
    Sprite Size Mode=::NewEnumerator3
    UsePositionOffset=true
    Write Material Random=true
   )
   &ShapeLocation(
    Cylinder Distribution=::NewEnumerator2
    Cylinder Height=200
    Cylinder Radius=1
    MaskSpawnGroup=false
    Offset Mode=::NewEnumerator0
    Shape Primitive=::NewEnumerator1
    UseOwnerScale=true
    V Position=&FloatFromCurve(
     FloatCurve=DataInterface() {propertyValues {Curve {keys [
         {interpMode=RCIM_Cubic, leaveTangent=-0.1288, tangentMode=RCTM_Break, time=0, value=0}
         {arriveTangent=2.2197, tangentMode=RCTM_User, time=0.3, value=1}
        ]}}}
     CurveIndex=&Add_Float(A=$Engine.System.Age, B=0)
    )
   ) {enabled=false}
   &InitialMeshOrientation(
    Mesh Axis To Orient=<0 0 -1>
    Mesh Orientation Mode=::NewEnumerator2
    Orientation Coordinate Space=::World
    Rotation=&RandomRangeVector(Minimum=<0 0 0>, Maximum=<0 0 1>)
    Use Rotation Vector=true
   )
   &CameraOffset(Camera Offset Amount=15)
   &AddVelocity(
    Rotation Coordinate Space=::World
    Velocity=<0 0 1>
    Velocity Speed Scale=&RandomRangeFloat(Minimum=-40, Maximum=-60)
   )
  ]
  particleUpdateScript [
   &ParticleState()
   /SetVariables(Particles.CamDistanceAlphaSplines:NiagaraFloat=&RemapRange(
     Input Value To Remap=/Niagara/DynamicInputs/Camera/DistanceToCamera(Use Owner Position:NiagaraBool=true, Camera Query:NiagaraDataInterfaceCamera=DataInterface())
     Input Min=100
     Input Max=200
     Output Min=0
     Output Max=1
    )) {isSetParameters=true}
   &ScaleMeshSize(Scale Factor=&MakeVector(X=1, Y=1, Z=0.8)) {module=ScaleMeshSize001}
   /SetVariables(Particles.XY:NiagaraFloat=&FloatFromCurve(FloatCurve=DataInterface() {propertyValues {Curve {
        defaultValue=1
        keys [
         {arriveTangent=0.1928, interpMode=RCIM_Cubic, leaveTangent=0.1928, tangentMode=RCTM_User, time=0.2, value=0.8}
         {arriveTangent=1.9927, interpMode=RCIM_Cubic, leaveTangent=1.9927, tangentMode=RCTM_User, time=1, value=2}
        ]
       }}})) {isSetParameters=true}
   &ScaleMeshSize(Scale Factor=&MakeVector(X=$Particles.XY, Y=$Particles.XY, Z=$Particles.XY))
   &UpdateMeshOrientation(Rotation Rate=&RandomRangeFloat(Minimum=0.2, Maximum=0.4), Rotation Vector=<0 0 1>)
   &ScaleColor(
    Scale Alpha=&FloatFromCurve(
     FloatCurve=DataInterface() {propertyValues {Curve {
        defaultValue=1
        keys [
         {arriveTangent=1.5203, interpMode=RCIM_Cubic, leaveTangent=1.5203, tangentMode=RCTM_User, time=0, value=0}
         <at 0.2 1 RCIM_Cubic>
         {arriveTangent=-1.9333, interpMode=RCIM_Cubic, leaveTangent=-1.9333, tangentMode=RCTM_User, time=1, value=0}
        ]
       }}}
     Scale Curve=&Multiply_Float(A=0.3, B=&Multiply_Float(A=$System.CurveDestroyed, B=$Particles.CamDistanceAlphaSplines))
    )
    Scale RGB=&MakeVectorFromLinearColorRGB(Color=&ColorFromCurve(ColorCurve=DataInterface() {propertyValues {
        AlphaCurve {keys [<at 0 1 RCIM_Cubic>]}
        BlueCurve {keys [<at 0.1729 2 RCIM_Cubic>,<at 0.8656 1 RCIM_Cubic>]}
        GreenCurve {keys [<at 0.1729 1.4667 RCIM_Cubic>,<at 0.8656 0.5 RCIM_Cubic>]}
        RedCurve {keys [<at 0.1729 0.4 RCIM_Cubic>,<at 0.8656 0.5833 RCIM_Cubic>]}
       }}))
    ScaleA=true
    ScaleRGB=true
   )
   &InheritSourceMovement(
    Applied Position Scale=&VectorFromFloat(Value=1)
    ApplyForce=false
    ApplyPosition=true
    ApplyPositionFalloff=false
   ) {enabled=false}
   &SolveForcesAndVelocity(Clamp Velocity=false, Limit Acceleration=false)
   &DynamicMaterialParameters(
    Index 0 Param 1=&FloatFromCurve(FloatCurve=DataInterface() {propertyValues {Curve {keys [<at 0 -0.5>,<at 1 0.3>]}}})
    Module.Index0Float0Write=true
    Module.Index0Float1Write=false
    Module.Index0Float2Write=false
    Module.Index0Float3Write=false
   )
   /SetVariables(Particles.ColorGlow:LinearColor=&MultiplyLinearColor(Linear Color=$Particles.Color, Scale Factor=&MakeVector4_FromLinearColor(LinearColor=<rgba 2 3 4 1>))) {isSetParameters=true}
   /SetVariables(Particles.ColorTip:LinearColor=&MultiplyLinearColor(Linear Color=$Particles.Color, Scale Factor=&MakeVector4_FromLinearColor(LinearColor=<rgba 0 3.3631 16 1>))) {enabled=false, isSetParameters=true}
  ]
  rendererClasses [&NiagaraMeshRendererProperties]
  renderers [
   {rendererClass=&NiagaraMeshRendererProperties, rendererIndex=0}
   {rendererClass=&NiagaraMeshRendererProperties, rendererIndex=1}
  ]
 }
 MarkerCore {
  emitterUpdateScript [
   &EmitterState(
    Inactive Response=::NewEnumerator1
    Life Cycle Mode=::NewEnumerator1
    Loop Behavior=::NewEnumerator1
    Loop Delay=1
    Loop Duration Mode=::NewEnumerator1
    UseLoopDelay=false
   )
   /SetVariables(Emitter.MeshOrientation:Quat4f=&MultiplyQuaternion(
     Quaternion A=<1 0 0 -1>
     Quaternion B=&MakeQuaternion(Quaternion From=0, Axis=<0 1 0>, Angle=0.25, From=?, To=?, Angle Type=::NewEnumerator1, Coordinate Space=::Local, Normalize Vectors=false)
    )) {isSetParameters=true}
   /SetVariables(
    Emitter.Color:LinearColor=&MultiplyLinearColor(
     Linear Color=&FNTimeOfDayColor(
      Night Color=<rgba 0.45 0.5 0.4917 1>
      Morning Color=<rgba 0.72 0.7733 0.8 1>
      Day Color=<rgba 0.9 1 0.9833 1>
      Evening Color=<rgba 0.63 0.6767 0.7 1>
     )
     Scale Factor=&MakeVector4FromVectorAndFloat(
      "Vector (RGB)"=<1 1 1>
      "Float (Alpha)"=&FloatFromCurve(
       FloatCurve=DataInterface() {propertyValues {Curve {keys [{interpMode=RCIM_Cubic, tangentMode=RCTM_User, time=0.3, value=0}, {interpMode=RCIM_Cubic, tangentMode=RCTM_User, time=0.7, value=1}]}}}
       CurveIndex=$System.FXFadeIn
       Scale Curve=&FloatFromCurve(FloatCurve=DataInterface() {propertyValues {Curve {keys [<at 0 0>,<at 0.4 1>]}}}, CurveIndex=$System.CamDistanceAlpha)
      )
     )
    )
    Emitter.ScalePortalCore:Vector3f=&Multiply_VectorByFloat(
     Vector=&Multiply_Vector(
      A=<1 1.3 1>
      B=&MakeVector(
       X=&FloatFromCurve(
        FloatCurve=DataInterface() {propertyValues {Curve {keys [
            {arriveTangent=-0.6813, interpMode=RCIM_Cubic, leaveTangent=-0.6813, tangentMode=RCTM_User, time=0, value=0.15}
            {arriveTangent=0.0618, interpMode=RCIM_Cubic, leaveTangent=0.0618, tangentMode=RCTM_User, time=0.3, value=1}
           ]}}}
        CurveIndex=$System.Age
       )
       Y=&FloatFromCurve(
        FloatCurve=DataInterface() {propertyValues {Curve {keys [
            {arriveTangent=5.8747, interpMode=RCIM_Cubic, leaveTangent=5.8747, tangentMode=RCTM_User, time=0, value=0.15}
            {interpMode=RCIM_Cubic, tangentMode=RCTM_User, time=0.35, value=1}
           ]}}}
        CurveIndex=$System.Age
       )
       Z=1
      )
     )
     Float=&Multiply_Float(A=&Multiply_Float(A=$System.FXScale, B=100), B=$System.CurveDestroyed)
    )
   ) {isSetParameters=true}
   /SetVariables(
    Emitter.GlowColor:LinearColor=&MultiplyLinearColor(
     Linear Color=&FNTimeOfDayColor(Night Color=<rgba 0.675 3.225 4.5 1>, Morning Color=<rgba 0.9 4.3 6 1>, Day Color=<rgba 1.5 7.875 10 1>, Evening Color=<rgba 0.9 4.3 6 1>)
     Scale Factor=&MakeVector4FromVectorAndFloat(
      "Vector (RGB)"=<1 1 1>
      "Float (Alpha)"=&FloatFromCurve(
       FloatCurve=DataInterface() {propertyValues {Curve {keys [<at 0.6 0>,<at 1 1>]}}}
       CurveIndex=$System.FXFadeIn
       Scale Curve=$System.CamDistanceAlpha
      )
     )
    )
    Emitter.ScaleGlow:Vector3f=&Multiply_VectorByFloat(Vector=$Emitter.ScalePortalCore, Float=1.18)
   ) {isSetParameters=true}
   /SetVariables(
    Emitter.EnergyColor:LinearColor=&MultiplyLinearColor(
     Linear Color=<rgba 0.5 1 0.9167 1>
     Scale Factor=&MakeVector4FromVectorAndFloat(
      "Vector (RGB)"=<1 1 1>
      "Float (Alpha)"=&Multiply_Float(
       A=&FNTimeOfDayScalar(Night Scale=0.66, Morning Scale=0.8, Evening Scale=0.8)
       B=&Multiply_Float(A=$System.FXFadeIn, B=&Multiply_Float(A=$System.CurveDestroyed, B=$System.CamDistanceAlpha))
      )
     )
    )
    Emitter.MeshOrientationEnergy:Quat4f=&MultiplyQuaternion(Quaternion A=$Emitter.MeshOrientation, Quaternion B=<0 1 0 1>)
    Emitter.ScaleEnergy:Vector3f=&Multiply_Vector(
     A=$Emitter.ScalePortalCore
     B=&Multiply_VectorByFloat(
      Vector=<0.0001 0.01 0.01>
      Float=&Multiply_Float(
       A=0.79
       B=&FloatFromCurve(FloatCurve=DataInterface() {propertyValues {Curve {keys [<at 0.7 0>,<at 1 1>]}}}, CurveIndex=$System.FXFadeIn)
      )
     )
    )
   ) {isSetParameters=true}
   /SetVariables(Emitter.LightEnabled:NiagaraBool=&SetBoolByFloatComparison(A=$Emitter.Age, B=0.5)) {isSetParameters=true}
   /SetVariables(Emitter.ColorLight:LinearColor=&MultiplyLinearColor(
     Linear Color=<rgba 8.4 36.2999 120 1>
     Scale Factor=&MakeVector4FromVectorAndFloat(
      "Vector (RGB)"=<1 1 1>
      "Float (Alpha)"=&Multiply_Float(
       A=$System.CurveDestroyed
       B=&FloatFromCurve(FloatCurve=DataInterface() {propertyValues {Curve {keys [<at 0.5 0>,<at 0.8 1>]}}}, CurveIndex=$Emitter.Age)
      )
     )
    )) {isSetParameters=true}
   /SetVariables(
    Emitter.ColorRadialWaves:LinearColor=&MultiplyLinearColor(
     Linear Color=<rgba 0.42 6.0783 14 1>
     Scale Factor=&MakeVector4FromVectorAndFloat(
      "Vector (RGB)"=<1 1 1>
      "Float (Alpha)"=&Multiply_Float(
       A=&FNTimeOfDayScalar(Night Scale=0.66, Morning Scale=0.8, Evening Scale=0.8)
       B=&Multiply_Float(A=&Multiply_Float(A=$System.CurveDestroyed, B=$System.FXFadeIn), B=$System.CamDistanceAlpha)
      )
     )
    )
    Emitter.ScaleRadialWaves:Vector3f=&Multiply_Vector(A=$Emitter.ScalePortalCore, B=&Multiply_VectorByFloat(Vector=<1 1.2 1>, Float=1.3))
   ) {isSetParameters=true}
   &SpawnRate(SpawnRate=2, Use Spawn Probability=false)
  ]
  particleSpawnScript [
   &InitializeParticle(
    Color=<rgba 0 106.6666 160 2>
    Lifetime=1.7
    Mesh Scale=&Multiply_VectorByFloat(Vector=<1 1 1.3>, Float=&Multiply_Float(A=$System.FXScale, B=0.55))
    Mesh Scale Mode=::NewEnumerator4
    UsePositionOffset=false
    Write Material Random=true
   )
   &InitialMeshOrientation(Mesh Orientation Mode=::NewEnumerator2, Use Rotation Vector=false)
  ]
  particleUpdateScript [
   &ParticleState()
   &ScaleMeshSize(Scale Factor=&Multiply_VectorByFloat(
     Vector=<1 1 1>
     Float=&Lerp_Float(
      A=0.93
      B=1.5
      Alpha=&FloatFromCurve(FloatCurve=DataInterface() {propertyValues {Curve {keys [
           {arriveTangent=-0.0304, interpMode=RCIM_Cubic, leaveTangent=-0.0304, tangentMode=RCTM_User, time=0, value=0}
           {arriveTangent=1.9689, interpMode=RCIM_Cubic, leaveTangent=1.9689, tangentMode=RCTM_User, time=1, value=1}
          ]}}})
     )
    ))
   &ScaleColor(
    Scale Alpha=&FloatFromCurve(
     FloatCurve=DataInterface() {propertyValues {Curve {keys [
         {interpMode=RCIM_Cubic, tangentMode=RCTM_User, time=0, value=0}
         {arriveTangent=-0.5761, interpMode=RCIM_Cubic, leaveTangent=-0.5761, tangentMode=RCTM_User, time=0.25, value=1}
         {arriveTangent=-1.6788, interpMode=RCIM_Cubic, leaveTangent=-1.6788, tangentMode=RCTM_User, time=1, value=0}
        ]}}}
     Scale Curve=&Multiply_Float(A=$System.CurveDestroyed, B=$System.CamDistanceAlpha)
    )
    Scale RGB=&VectorFromFloat(Value=&FloatFromCurve(FloatCurve=DataInterface() {propertyValues {Curve {keys [{interpMode=RCIM_Cubic, tangentMode=RCTM_User, time=0.2, value=1}, {interpMode=RCIM_Cubic, tangentMode=RCTM_User, time=1, value=0.05}]}}}))
    ScaleA=true
    ScaleRGB=true
   )
   &DynamicMaterialParameters(
    Index 0 Param 1=&Lerp_Float(
     A=0
     B=&RandomRangeFloat(Minimum=-0.2, Maximum=0.2)
     Alpha=&FloatFromCurve(FloatCurve=DataInterface() {propertyValues {Curve {keys [
          {arriveTangent=1.212, interpMode=RCIM_Cubic, leaveTangent=1.212, tangentMode=RCTM_User, time=0, value=0}
          {interpMode=RCIM_Cubic, tangentMode=RCTM_User, time=0.8, value=1}
         ]}}})
    )
    Index 0 Param 2=&FloatFromCurve(FloatCurve=DataInterface() {propertyValues {Curve {keys [
         {interpMode=RCIM_Cubic, tangentMode=RCTM_User, time=0, value=0}
         {arriveTangent=0.5428, interpMode=RCIM_Cubic, leaveTangent=0.5428, tangentMode=RCTM_User, time=0.2, value=0.5}
         {arriveTangent=3.6615, tangentMode=RCTM_User, time=1, value=2}
        ]}}})
    Module.Index0Float0Write=true
    Module.Index0Float1Write=true
    Module.Index0Float2Write=false
    Module.Index0Float3Write=false
   )
  ]
  rendererClasses [&NiagaraMeshRendererProperties, &NiagaraLightRendererProperties]
  renderers [
   {rendererClass=&NiagaraMeshRendererProperties, rendererIndex=0}
   {rendererClass=&NiagaraMeshRendererProperties, rendererIndex=1}
   {rendererClass=&NiagaraMeshRendererProperties, rendererIndex=2}
   {rendererClass=&NiagaraLightRendererProperties, rendererIndex=3}
   {rendererClass=&NiagaraMeshRendererProperties, rendererIndex=4}
   {rendererClass=&NiagaraMeshRendererProperties, rendererIndex=5}
  ]
 }
 Sparkles {
  simTarget=GPUComputeSim
  emitterUpdateScript [
   &EmitterState(
    Enable Distance Culling=true
    Life Cycle Mode=::NewEnumerator1
    Loop Behavior=::NewEnumerator1
    Loop Duration Mode=::NewEnumerator1
    MaxDistance=2000
    Module.bUseMaxDistance=true
    Module.bUseMinDistance=false
    Scalability Mode=::NewEnumerator1
    UseLoopDelay=false
   )
   &SpawnRate(Spawn Probability=0.7, SpawnRate=12, Use Spawn Probability=true)
  ]
  particleSpawnScript [
   &InitializeParticle(
    Color=<rgba 1.5 1.5 1.5 1>
    Lifetime=&RandomRangeFloat(Minimum=0.7, Maximum=2)
    Sprite Rotation Mode=::NewEnumerator2
    Sprite Size Mode=::NewEnumerator1
    Uniform Sprite Size Max=8
    Uniform Sprite Size Min=2
    UsePositionOffset=false
    Write Material Random=true
   )
   &ShapeLocation(
    MaskSpawnGroup=false
    Non Uniform Scale=<1 1 1.3>
    Offset=$System.SystemOffset
    Offset Mode=::NewEnumerator0
    Override Seed=false
    Sphere Radius=&Multiply_Float(A=50, B=$System.FXScale)
    Sphere Surface Distribution=0.9
    UseOwnerScale=true
   )
   &AddVelocity(
    Module.Use Falloff Distance=false
    Origin Offset=$System.SystemOffset
    Override Seed=false
    Velocity Mode=::NewEnumerator1
    Velocity Speed=&RandomRangeFloat(Minimum=3, Maximum=10)
   )
   &CameraOffset(Camera Offset Amount=20)
  ]
  particleUpdateScript [
   &ParticleState()
   &ScaleSpriteSize(Uniform Curve Sprite Scale=DataInterface() {propertyValues {Curve {keys [
        <at 0 0 RCIM_Cubic>
        {interpMode=RCIM_Cubic, tangentMode=RCTM_User, time=0.2, value=1}
        {interpMode=RCIM_Cubic, tangentMode=RCTM_User, time=1, value=0}
       ]}}})
   &ScaleColor(
    Scale Alpha=&FloatFromCurve(
     FloatCurve=DataInterface() {propertyValues {Curve {keys [
         {interpMode=RCIM_Cubic, tangentMode=RCTM_User, time=0, value=0}
         {interpMode=RCIM_Cubic, tangentMode=RCTM_User, time=0.3, value=1}
         {interpMode=RCIM_Cubic, tangentMode=RCTM_User, time=1, value=0}
        ]}}}
     Scale Curve=&Multiply_Float(A=2, B=&Multiply_Float(A=$System.CurveDestroyed, B=$System.CamDistanceAlpha))
    )
    Scale RGB=&MakeVectorFromLinearColorRGB(Color=&ColorFromCurve(ColorCurve=DataInterface() {propertyValues {
        AlphaCurve {keys [<at 0 1 RCIM_Cubic>]}
        BlueCurve {keys [<at 0.0771 1 RCIM_Cubic>,<at 0.6657 1 RCIM_Cubic>]}
        GreenCurve {keys [<at 0.0771 0.6918 RCIM_Cubic>,<at 0.6657 1 RCIM_Cubic>]}
        RedCurve {keys [<at 0.0771 0.2643 RCIM_Cubic>,<at 0.6657 1 RCIM_Cubic>]}
       }}))
    ScaleA=true
    ScaleRGB=true
   )
   /SetVariables(
    Particles.GlowColor=&MultiplyLinearColor(
     Linear Color=$Particles.Color
     Scale Factor=&MakeVector4_FromLinearColor(LinearColor=<rgba 0.34 0.34 1 0.2>)
    )
    Particles.GlowSize=&Multiply_Vector2D_ByFloat(Float=2.2, Vector2D=$Particles.SpriteSize)
   ) {isSetParameters=true}
   &Drag(Drag=&RandomRangeFloat(Minimum=1, Maximum=2), Use Linear Drag=true, Use Rotational Drag=true)
   &GravityForce(Gravity=&Multiply_VectorByFloat(Vector=<0 0 -100>, Float=&RandomRangeFloat(Minimum=0.66, Maximum=1.2)))
   &VortexForce(
    Origin Pull Amount=200
    UseOriginOffset=false
    Vortex Axis=&ReturnCameraVector(Camera Vectors=::NewEnumerator1, Coordinate Space=::Local)
    Vortex Force Amount=300
   ) {enabled=false}
   &SolveForcesAndVelocity(Clamp Velocity=false, Limit Acceleration=false)
  ]
  rendererClasses [&NiagaraSpriteRendererProperties]
  renderers [
   {rendererClass=&NiagaraSpriteRendererProperties}
   {rendererClass=&NiagaraSpriteRendererProperties, rendererIndex=1}
  ]
 }
}
summary {
 systemName=NS_WinterDecay_Mask_Portal
 emitters [
  {emitterName=ExtraWaves, rendererClasses [&NiagaraMeshRendererProperties]}
  {emitterName=Sparkles, simTarget=GPUComputeSim, rendererClasses [&NiagaraSpriteRendererProperties]}
  {
   bEnabled=false
   emitterName=EnergyRadial
   rendererClasses [&NiagaraSpriteRendererProperties, &NiagaraMeshRendererProperties]
  }
  {emitterName=Helix, rendererClasses [&NiagaraMeshRendererProperties]}
  {emitterName=MarkerCore, rendererClasses [&NiagaraMeshRendererProperties, &NiagaraLightRendererProperties]}
 ]
 userVariables [
  User.bIsFrontEnd {type=NiagaraBool}
  User.bIsFrontEndPreview {type=NiagaraBool}
  User.Destroyed {type=NiagaraBool}
  User.Variant {type=NiagaraFloat, defaultValue=0, description=""}
 ]
}
