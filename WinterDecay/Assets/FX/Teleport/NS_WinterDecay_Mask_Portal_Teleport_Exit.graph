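# Niagara system for the exit burst of the WinterDecay mask portal teleport, judging by its name.
# The active Flare emitter bursts a short-lived, camera-offset sprite sized by System.FXScale and distance-culled.
# Over its life the flare widens, flattens and fades while curves cool its HDR color from bright cyan to dimmer blue.
# A disabled NE_BlackHole emitter retains grouped bursts with glow and ripple attributes plus source-movement inheritance.
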NiagaraSystem /WinterDecay/Assets/FX/Teleport/NS_WinterDecay_Mask_Portal_Teleport_Exit
dependencies {
 usedDataInterfaces [&NiagaraDataInterfaceCurve, &NiagaraDataInterfaceColorCurve]
 usedDynamicInputs []
 usedModules [
  &self:SystemSpawnScript.SystemScriptSource.SystemScriptGraph.NiagaraNodeAssignment_1.SetVariables_#1
  &self:SystemSpawnScript.SystemScriptSource.SystemScriptGraph.NiagaraNodeAssignment_0.SetVariables_#2
  &self:SystemSpawnScript.SystemScriptSource.SystemScriptGraph.NiagaraNodeAssignment_2.SetVariables_#3
  &SystemState
 ]
}
emitters {
 Flare {
  emitterUpdateScript [
   &EmitterState(
    Enable Distance Culling=true
    Life Cycle Mode=::NewEnumerator1
    Loop Behavior=::NewEnumerator1
    Loop Duration=1
    MaxDistance=4000
    Module.bUseMaxDistance=true
    Module.bUseMinDistance=false
    Scalability Mode=::NewEnumerator1
    UseLoopDelay=false
   )
   &SpawnBurst_Instantaneous(Use Spawn Probability=false, UseLoopCountLimit=false)
  ]
  particleSpawnScript [
   &InitializeParticle(
    Color=<rgba 0.6 0.82 1 1>
    Lifetime=0.33
    Sprite Size=&Multiply_Vector2D_ByFloat(Float=&Multiply_Float(A=$System.FXScale, B=300), Vector2D=<1 1.3>)
    Sprite Size Mode=::NewEnumerator4
    UsePositionOffset=false
    Write Material Random=true
   )
   &CameraOffset(Camera Offset Amount=30)
  ]
  particleUpdateScript [
   &ParticleState()
   &ScaleSpriteSize(
    Scale Factor=&MakeVector2D(
     X=&FloatFromCurve(FloatCurve=DataInterface() {propertyValues {Curve {keys [{interpMode=RCIM_Cubic, tangentMode=RCTM_User, time=0.3, value=1}, {interpMode=RCIM_Cubic, tangentMode=RCTM_User, time=1, value=1.5}]}}})
     Y=&FloatFromCurve(FloatCurve=DataInterface() {propertyValues {Curve {keys [<at 0 1>,<at 1 0.5>]}}})
    )
    Scale Sprite Size Mode=::NewEnumerator2
   ) {module=ScaleSpriteSize001}
   &ScaleSpriteSize()
   &ScaleColor(
    Scale Alpha=&FloatFromCurve(FloatCurve=DataInterface() {propertyValues {Curve {keys [<at 0 1>,<at 1 0>]}}})
    Scale RGB=&MakeVectorFromLinearColorRGB(Color=&ColorFromCurve(ColorCurve=DataInterface() {propertyValues {
        AlphaCurve {keys [<at 0 1 RCIM_Cubic>]}
        BlueCurve {keys [<at 0.1113 10 RCIM_Cubic>,<at 0.5326 2 RCIM_Cubic>]}
        GreenCurve {keys [<at 0.1113 10 RCIM_Cubic>,<at 0.5326 0.8075 RCIM_Cubic>]}
        RedCurve {keys [<at 0.1113 0.7 RCIM_Cubic>,<at 0.5326 0.3496 RCIM_Cubic>]}
       }}))
    ScaleA=true
    ScaleRGB=true
   )
   &DynamicMaterialParameters(
    Index 0 Param 1=1
    Index 0 Param 2=1
    Module.Index0Float0Write=true
    Module.Index0Float1Write=true
    Module.Index0Float2Write=false
    Module.Index0Float3Write=false
   )
  ]
  rendererClasses [&NiagaraSpriteRendererProperties]
  renderers [{rendererClass=&NiagaraSpriteRendererProperties}]
 }
 NE_BlackHole {
  bEnabled=false
  emitterUpdateScript [
   &EmitterState(
    Inactive Response=::NewEnumerator1
    Life Cycle Mode=::NewEnumerator1
    Loop Behavior=::NewEnumerator1
    Loop Delay=0.1
    Loop Duration Mode=::NewEnumerator1
    UseLoopDelay=true
   )
   &SpawnBurst_Instantaneous(Use Spawn Probability=false, UseLoopCountLimit=false)
   &SpawnBurst_Instantaneous(Spawn Group=1, Use Spawn Probability=false, UseLoopCountLimit=false) {module=SpawnBurst_Instantaneous001}
   &SpawnBurst_Instantaneous(Spawn Group=2, Use Spawn Probability=false, UseLoopCountLimit=false) {module=SpawnBurst_Instantaneous002}
  ]
  particleSpawnScript [
   /SetVariables(Particles.VisibilityTag=&SpawnGroup()) {isSetParameters=true}
   &InitializeParticle(
    Color=&MergeHelper_ColorSwitch(
     0=<rgba 1.02 3 2.67 0.2>
     1=<rgba 0.5 41.75 50 0.1>
     2=<rgba 0.63 0.7 0.7 0.66>
     3=?
     Index=&SpawnGroup()
     Num=::NewEnumerator1
    )
    Lifetime=0.25
    Mesh Scale=&Multiply_VectorByFloat(
     Vector=<1 1 1.5>
     Float=&MergeHelper_FloatSwitch(0=4.5, 1=3, 2=?, 3=?, Index=&SpawnGroup(), Num=::NewEnumerator0)
    )
    Mesh Scale Mode=::NewEnumerator4
    Position Offset=$System.FXPosition
    Sprite Size=&Multiply_Vector2D_ByFloat(Float=&Multiply_Float(A=120, B=$System.FXScale), Vector2D=<1 1>)
    Sprite Size Mode=::NewEnumerator4
    UsePositionOffset=false
    Write Material Random=true
   )
   &InitialMeshOrientation(Mesh Orientation Mode=::NewEnumerator2, Rotation=<0 0 0.25>, Use Rotation Vector=true)
   &CameraOffset() {enabled=false}
   &SpriteFacingAndAlignment(Sprite Alignment=<0 0 1>, WriteAlignment=true, WriteFacing=false)
   /SetVariables(
    Particles.GlowColor=<rgba 0.2 0.8667 1 0.66>
    Particles.GlowSize=&Multiply_Vector2D_ByFloat(Float=220, Vector2D=<0.7 1.1>)
   ) {isSetParameters=true}
   /SetVariables(
    Particles.RippleColor:LinearColor=<rgba 0.2 1.7 2 1>
    Particles.RippleSize:Vector2f=&Multiply_Vector2D_ByFloat(Float=&Multiply_Float(A=$System.FXScale, B=180), Vector2D=<0.7 1>)
   ) {isSetParameters=true}
  ]
  particleUpdateScript [
   &ParticleState()
   &ScaleSpriteSize(Uniform Curve Sprite Scale=DataInterface() {propertyValues {Curve {keys [
        {arriveTangent=3.511, interpMode=RCIM_Cubic, leaveTangent=3.511, tangentMode=RCTM_User, time=0, value=1}
        {arriveTangent=-0.2319, interpMode=RCIM_Cubic, leaveTangent=-0.2319, tangentMode=RCTM_User, time=1, value=0}
       ]}}})
   /SetVariables(Particles.ScaleSmall:Vector3f=&Multiply_VectorByFloat(Vector=$Particles.Scale, Float=1.2)) {isSetParameters=true}
   &ScaleColor(
    Scale RGB=&MakeVectorFromLinearColorRGB(Color=&MergeHelper_ColorSwitch(
      0=&FNTimeOfDayColor(
       Night Color=<rgba 0 0.4754 0.5 0.8>
       Morning Color=<rgba 0.2381 0.1011 1 1>
       Day Color=<rgba 1 1 1 1>
       Evening Color=<rgba 0.2381 0.1011 1 1>
      )
      1=&FNTimeOfDayColor(Night Color=<rgba 1.5 0.0237 1.2607 1>, Morning Color=<rgba 1 1 1 1>, Day Color=<rgba 1 1 1 1>, Evening Color=<rgba 1 1 1 1>)
      2=&FNTimeOfDayColor(Night Color=<rgba 0.581 1.0995 1.5 1>, Morning Color=<rgba 1 1 1 1>, Day Color=<rgba 1 1 1 1>, Evening Color=<rgba 1.2 1.2 1.2 1>)
      3=?
      Index=&SpawnGroup()
      Num=::NewEnumerator1
     ))
    ScaleA=true
    ScaleRGB=true
   ) {enabled=false}
   &InheritSourceMovement(
    Applied Position Scale=&VectorFromFloat(Value=&FloatFromCurve(FloatCurve=DataInterface() {propertyValues {Curve {keys [<at 0 1>,<at 0.5 0>]}}}))
    ApplyForce=false
    ApplyPosition=true
    ApplyPositionFalloff=false
   )
   &SolveForcesAndVelocity(Clamp Velocity=false, Limit Acceleration=false)
  ]
  rendererClasses [&NiagaraSpriteRendererProperties]
  renderers [{rendererClass=&NiagaraSpriteRendererProperties}]
 }
}
summary {
 systemName=NS_WinterDecay_Mask_Portal_Teleport_Exit
 emitters [
  {bEnabled=false, emitterName=NE_BlackHole, rendererClasses [&NiagaraSpriteRendererProperties]}
  {emitterName=Flare, rendererClasses [&NiagaraSpriteRendererProperties]}
 ]
 userVariables [
  User.bIsFrontEnd {type=NiagaraBool}
  User.bIsFrontEndPreview {type=NiagaraBool}
  User.Variant {type=NiagaraFloat, defaultValue=0, description=""}
 ]
}
